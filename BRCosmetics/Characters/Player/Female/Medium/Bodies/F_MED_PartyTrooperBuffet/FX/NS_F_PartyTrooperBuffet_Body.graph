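# Niagara body effect for the PartyTrooperBuffet cosmetic character, spawning looping GPU sprites sampled on the body skeletal mesh.
# A body FX texture, read at each sampled UV, culls particles by red-channel threshold to confine sprites to painted areas.
# An inverted region mask, shifted when the foot-hidden user flag is set, modulates alpha.
# Curves pulse sprite size and fade alpha, with a slight camera offset.
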NiagaraSystem /BRCosmetics/Characters/Player/Female/Medium/Bodies/F_MED_PartyTrooperBuffet/FX/NS_F_PartyTrooperBuffet_Body
dependencies {
 usedDataInterfaces [&NiagaraDataInterfaceCurve, /Script/Niagara.NiagaraDataInterfaceTexture]
 usedDynamicInputs []
 usedModules [&SystemState]
}
emitters {OmnidirectionalBurst {
  simTarget=GPUComputeSim
  emitterUpdateScript [
   &EmitterState(Life Cycle Mode=::NewEnumerator1, Loop Duration=15, UseLoopDelay=false)
   &SpawnRate(SpawnRate=18, Use Spawn Probability=false)
  ]
  particleSpawnScript [&InitializeParticle(
    Lifetime Max=3
    Lifetime Mode=::NewEnumerator1
    Sprite Size Mode=::NewEnumerator3
    Uniform Sprite Size=&RandomRangeFloat(Minimum=18, Maximum=25)
    UsePositionOffset=false
    Write Material Random=true
   )]
  particleUpdateScript [
   &ParticleState()
   &ScaleSpriteSize(Scale Factor=&Vector2DFromFloat(Value=&FloatFromCurve(FloatCurve=DataInterface() {propertyValues {Curve {keys [
          {arriveTangent=4.1465, interpMode=RCIM_Cubic, leaveTangent=4.1465, tangentMode=RCTM_User, time=0, value=0.15}
          <at 0.2 1>
          {arriveTangent=-0.6938, interpMode=RCIM_Cubic, leaveTangent=-0.6938, tangentMode=RCTM_User, time=0.5, value=0.35}
          <at 1.3 0>
         ]}}})))
   &SkeletalMeshLocation(
    "Interpolate Mesh Sampling (Update)"=true
    MaskSpawnGroup=false
    Mesh Sampling Type=::NewEnumerator7
    OffsetAlongNormal=false
    Orientation Sampling=::NewEnumerator1
    OverrideRandomSeed=false
    Randomness Mode:ENiagaraRandomnessMode=::NewEnumerator1
    Skeletal Mesh=$User.SkelMesh_Body
   )
   &RegionMask(
    Invert Region=true
    Offset:Vector3f=&MakeCustomVectorFromBool(Bool=$User.bisFootHidden, True Vector=<0 0 50>, False Vector=<0 0 0>)
    Offset Space:ENiagaraCoordinateSpace=::Local
    Region Type:ENiagaraDebugDrawShape=::NewEnumerator2
    Rotation:Vector3f=<0 0.25 0>
    UseOffset=true
    UseRotation=false
   )
   &ScaleColor(
    Scale Alpha=&FloatFromCurve(
     FloatCurve=DataInterface() {propertyValues {Curve {keys [<at 0 0.1 RCIM_Cubic>,<at 0.2 0.9>,<at 1.0011 0.002 RCIM_Cubic>]}}}
     Scale Curve=&Multiply_Float(A=3, B=$Particles.RegionMask.Float)
    )
    ScaleA=true
    ScaleRGB=true
   )
   /Niagara/Modules/Update/Textures/SampleTexture(
    Texture=DataInterface() {propertyValues {
      Texture=/BRCosmetics/Characters/Player/Female/Medium/Bodies/F_MED_Ch_One_Redux/Skins/PartyTrooperBuffet/Textures/T_F_MED_Party_Trooper_Buffet_Body_FX
      TextureUserParameter {parameter {name=None, typeDefHandle {registeredTypeIndex=2}, varData []}}
     }}
    UV:Vector2f=$Particles.SkeletalMeshLocation.SampledUV
   ) {module=SampleTexture001}
   &KillParticles(Kill Particles=&SetBoolByFloatComparison(
     A=&MakeFloatFromLinearColor(LinearColor=$Particles.SampleTexture001.SampledColor, Channel:ENiagaraLinearColor_Channels=::NewEnumerator0)
     B=0.9
     Comparison Type=::NewEnumerator6
    )) {module=KillParticles001}
   &KillParticles(Kill Particles=&Boolean_NotOperation(Bool=$Particles.RegionMask.IsInsideMask)) {enabled=false}
   &CameraOffset(Camera Offset Amount=5)
  ]
  rendererClasses [&NiagaraSpriteRendererProperties]
  renderers [{rendererClass=&NiagaraSpriteRendererProperties}]
 }}
summary {
 systemName=NS_F_PartyTrooperBuffet_Body
 emitters [{emitterName=OmnidirectionalBurst, simTarget=GPUComputeSim, rendererClasses [&NiagaraSpriteRendererProperties]}]
 userVariables [
  User.bisFootHidden {type=NiagaraBool}
  User.SkelMesh_Body {
   type=NiagaraDataInterfaceSkeletalMesh
   defaultValue {
    struct=NiagaraExt_VariableValue_DataInterface
    value {dataInterface=&self:NiagaraDataInterfaceSkeletalMesh_0}
   }
  }
 ]
}
