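# Niagara system for the PartyTrooperBuffet character cosmetic that shows cracked, dripping eyes as a timed damage reaction.
# Eye_Cracked bursts four socket-attached sprites on User.SkelMesh, with scratch modules offsetting and rotating them and a curve popping their size.
# Eye_Drip spawns falling tear sprites that inherit socket velocity and animate a dynamic material parameter late in life.
# User.DamageTime and User.DamageDuration feed both emitters, with curves ramping crack offsets and drip rate.
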NiagaraSystem /BRCosmetics/Characters/Player/Female/Medium/Bodies/F_MED_PartyTrooperBuffet/FX/NS_F_PartyTrooperBuffet_CrackEyes
dependencies {
 usedDataInterfaces [&NiagaraDataInterfaceCurve, &NiagaraDataInterfaceArrayFloat]
 usedDynamicInputs []
 usedModules [&SystemState, &Timeline]
}
emitters {
 Eye_Cracked {
  emitterUpdateScript [
   &EmitterState(Inactive Response=::NewEnumerator2, Life Cycle Mode=::NewEnumerator1, UseLoopDelay=false)
   &SpawnBurst_Instantaneous(Spawn Count=4, Use Spawn Probability=false)
  ]
  particleSpawnScript [
   &InitializeParticle(
    Color Mode=::NewEnumerator0
    Lifetime=15
    Position Offset=<0 0 125>
    Sprite Rotation Mode=::NewEnumerator0
    Sprite Size Mode=::NewEnumerator3
    Uniform Sprite Size=6.75
    UsePositionOffset=false
    Write Material Random=true
   )
   &self:Init_EyeOffsets(EyeCrackAngle=-8, EyeCrackShift=0.3) {module=Init_EyeOffsets}
   &SampleSkeletalMesh(
    AdditionalYawPitchRoll=false
    Bone / Socket Index:NiagaraInt32=$Particles.EyeSocketIdx
    Mesh Sampling Type:ENiagaraSkelSamplingModeFull=::NewEnumerator5
    Skeletal Mesh:NiagaraDataInterfaceSkeletalMesh=$User.SkelMesh
    Socket Sampling Mode:ENiagaraSocketSamplingMode=::NewEnumerator5
   )
   &CameraOffset(Camera Offset Amount=&Multiply_Float(A=&MakeFloatFromVector(Vector=$Engine.Owner.Scale, Channel=::NewEnumerator0), B=1))
  ]
  particleUpdateScript [
   &ParticleState()
   &SkeletalMeshLocation(
    Bone / Socket Index=$Particles.EyeSocketIdx
    "Interpolate Mesh Sampling (Update)"=true
    MaskSpawnGroup=false
    Mesh Sampling Type=::NewEnumerator5
    Sampled Position Offset Coordinate Space:ENiagaraSkelMeshTransforms=::NewEnumerator0
    Skeletal Mesh=$User.SkelMesh
    Socket Sampling Mode=::NewEnumerator5
    Use Sampled Mesh Attribute as Ribbon ID=false
   )
   &self:Update_EyeOffsets(
    DamageDuration:NiagaraFloat=$User.DamageDuration
    DamageTime:NiagaraFloat=$User.DamageTime
    DamageTimeRamped:NiagaraFloat=&FloatFromCurve(
     FloatCurve=DataInterface() {propertyValues {Curve {keys [
         {interpMode=RCIM_Cubic, tangentMode=RCTM_User, time=0, value=0}
         {arriveTangent=5.4924, arriveTangentWeight=0.0359, leaveTangent=5.4924, leaveTangentWeight=0.0248, time=0.03, value=1}
         {arriveTangentWeight=0.0607, time=0.06, value=1}
         <at 0.98 1>
         <at 1 0>
        ]}}}
     CurveIndex=$Particles.DamageTimeElapsed
    )
    Eye Halves Swivel Z=-29.5269
    EyeCrackShift=0.7
    EyeOffset:Vector3f=<2.4 0.15 0>
    "Eyes LR Fold  Z"=-12
    Eyes LR Tilt Y=5.8821
    SM_SampledRotation:Quat4f=$Particles.SkeletalMeshLocation.SampledRotation
   ) {module=Update_EyeOffsets}
   /SetVariables(Particles.SpriteSize=&Multiply_Vector2D_ByFloat(
     Float=&Multiply_Float(
      A=&FloatFromCurve(
       FloatCurve=DataInterface() {propertyValues {Curve {keys [
           {arriveTangent=2.8837, interpMode=RCIM_Cubic, leaveTangent=2.8837, tangentMode=RCTM_User, time=0, value=1}
           {arriveTangent=-3.0766, interpMode=RCIM_Cubic, leaveTangent=-3.0766, tangentMode=RCTM_User, time=0.006, value=1.1241}
           {arriveTangent=-0.1031, interpMode=RCIM_Cubic, leaveTangent=-0.1031, tangentMode=RCTM_User, time=0.0403, value=0.9393}
           <at 0.0617 1.0289>
           {tangentMode=RCTM_User, time=0.075, value=0.975}
          ]}}}
       CurveIndex=$Particles.DamageTimeElapsed
      )
      B=1
     )
     Vector2D=$Particles.Initial.SpriteSize
    )) {isSetParameters=true}
   &ApplyOwnerScaleToAttributes(Scale Initial Sprite Size=true)
  ]
  rendererClasses [&NiagaraSpriteRendererProperties]
  renderers [{rendererClass=&NiagaraSpriteRendererProperties}]
 }
 Eye_Drip {
  emitterUpdateScript [
   &EmitterState()
   &self:SetDamageTimeElapsed(DamageTime:NiagaraFloat=$User.DamageTime, DamageTimeLength:NiagaraFloat=$User.DamageDuration) {module=SetDamageTimeElapsed}
   &SpawnRate(
    SpawnRate=&Multiply_Float(
     A=4
     B=&FloatFromCurve(FloatCurve=DataInterface() {propertyValues {Curve {keys [<at 0 0>,<at 0.02 1>,<at 0.9 1>,<at 1 0>]}}}, CurveIndex=$Emitter.DamageTimeElapsed)
    )
    Use Spawn Probability=false
   )
   &SpawnPerFrame(Spawn:NiagaraBool=&SetBoolByFloatComparison(A=0, B=0.5), Use Spawn Probability=false) {enabled=false}
  ]
  particleSpawnScript [
   &InitializeParticle(
    Color Mode=::NewEnumerator0
    Lifetime=0.9
    Position Offset=<0 0 125>
    Sprite Rotation Mode=::NewEnumerator0
    Sprite Size Mode=::NewEnumerator3
    Uniform Sprite Size=2.75
    UsePositionOffset=false
    Write Material Random=true
   )
   &self:Set_TearOffsets(EyeCrackShift=1, EyeOffset:Vector3f=<0 25 -35>) {module=Set_TearOffsets}
   &SampleSkeletalMesh(
    AdditionalYawPitchRoll=false
    Bone / Socket Index:NiagaraInt32=$Particles.EyeSocketIdx
    Interpolate Mesh Sampling=false
    Mesh Sampling Type:ENiagaraSkelSamplingModeFull=::NewEnumerator5
    Sampled Position Offset:Vector3f=$Particles.EyeOffset
    Skeletal Mesh:NiagaraDataInterfaceSkeletalMesh=$User.SkelMesh
    Socket Sampling Mode:ENiagaraSocketSamplingMode=::NewEnumerator5
   )
   &self:Update_EyeOffsets(
    DamageTime:NiagaraFloat=$User.DamageTime
    EyeCrackAngle=0
    EyeOffset:Vector3f=<0 0 -15>
    SM_SampledRotation:Quat4f=$Particles.SampleSkeletalMesh.SampledRotation
   ) {module=Update_EyeOffsets}
   &SkeletalMeshLocation(
    Bone / Socket Index=$Particles.EyeSocketIdx
    "Inherit Velocity (Scale)"=0.4
    MaskSpawnGroup=false
    Mesh Sampling Type=::NewEnumerator5
    Position Sampling=::NewEnumerator1
    Sampled Position Offset=$Particles.EyeOffset
    Sampled Position Offset Coordinate Space:ENiagaraSkelMeshTransforms=::NewEnumerator0
    Skeletal Mesh=$User.SkelMesh
    Socket Sampling Mode=::NewEnumerator5
    Velocity Sampling=::NewEnumerator0
    VelocitySpeedLimit=false
   )
   &AddVelocity(
    Coordinate Space=::Simulation
    Velocity=&RotateVectorByQuaternion(Vector To Rotate=<0 10 0>, Quaternion=$Particles.SampleSkeletalMesh.SampledRotation)
   )
   &self:ScratchModule(offset:Vector3f=<0 0 -3>) {module=ScratchModule}
  ]
  particleUpdateScript [
   &ParticleState()
   &GravityForce(Gravity=<0 0 -90>)
   &SolveForcesAndVelocity(Clamp Velocity=false, Limit Acceleration=false)
   &DynamicMaterialParameters(
    Index 3 Param 4=&FloatFromCurve(FloatCurve=DataInterface() {propertyValues {Curve {keys [
         {arriveTangent=-6.0786, interpMode=RCIM_Cubic, leaveTangent=-6.0786, tangentMode=RCTM_User, time=0.6, value=0.85}
         {arriveTangent=-6.8417, interpMode=RCIM_Cubic, leaveTangent=-6.8417, tangentMode=RCTM_User, time=0.7, value=0.1}
        ]}}})
    Module.Param0WriteEnabled3=false
    Module.Param1WriteEnabled3=false
    Module.Param2WriteEnabled3=false
    Module.Param3WriteEnabled3=true
    Write Parameter Index 0=false
    Write Parameter Index 3=true
   )
  ]
  rendererClasses [&NiagaraSpriteRendererProperties]
  renderers [{rendererClass=&NiagaraSpriteRendererProperties}]
 }
}
summary {
 systemName=NS_F_PartyTrooperBuffet_CrackEyes
 emitters [
  {emitterName=Eye_Cracked, rendererClasses [&NiagaraSpriteRendererProperties]}
  {emitterName=Eye_Drip, rendererClasses [&NiagaraSpriteRendererProperties]}
 ]
 userVariables [
  User.DamageDuration {type=NiagaraFloat, defaultValue=5, description=""}
  User.DamageTime {type=NiagaraFloat, defaultValue=0, description=""}
  User.SkelMesh {
   type=NiagaraDataInterfaceSkeletalMesh
   defaultValue {
    struct=NiagaraExt_VariableValue_DataInterface
    value {dataInterface=&self:NiagaraDataInterfaceSkeletalMesh_0}
   }
  }
 ]
}
